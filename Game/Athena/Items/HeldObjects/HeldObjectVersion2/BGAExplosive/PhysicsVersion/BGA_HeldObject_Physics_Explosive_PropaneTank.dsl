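; Physics-driven propane tank held object that catches fire from large impacts and can rocket around before it dies.
; On the server, a large impulse triggers SetonFire once, behind a gate that closes briefly after throws.
; If thrust is enabled, a fire hit adds a thruster nudged toward center of mass, drops friction, then applies timed repeating impulses late in the fuse.
; Entering water drives floating Niagara effects; death stops thrust and traces for water to spawn a splash.
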(event EventBeginPlay
  ;[1 1 1] Guard against exploding off impulse, if the impulse is gotten right after the throw event.
  (if (Networking|IsServer)
    (EventDispatchers|AssignOnHeldObjectThrown (Variables|Default|GetFortHeldObject) (delegate Custom|OnHeldObjectThrown))
    (else
      _))
  (|CallParentFunction)
  (bind _returnvalue (BuildingActor|PhysicsSimulationSupport|GetPhysicsObjectComponent self))
  (PhysicsObject|BindEventtoOnReceivedLargeImpulse _returnvalue (EventDispatchers|CreateEvent :!event Custom|Handle%20Received%20Large%20Impulse)))

(event Custom|OnHeldObjectThrown
  ;[1 1 1] Guard against exploding off impulse, if the impulse is gotten right after the throw event.
  (CallFunction|CloseExplosionGate)
  (bind _returnvalue (Utilities|Time|SetTimerbyEvent (delegate Custom|OpenExplosionGate) (Variables|Default|GetTime_ThrownIgnoreImpulse)))
  (Variables|Default|SetTimer_ThrownIgnoreImpulse _returnvalue))

(event Custom|OpenExplosionGate
  ;[1 1 1] When a big impulse is taken on the server set on fire
  (Utilities|FlowControl|Gate false
    (:Exit
      (if (Networking|IsServer)
        (Utilities|FlowControl|DoOnce false
          (:Completed
            (CallFunction|SetonFire)))))))

(event Custom|CloseExplosionGate
  ;[1 1 1] When a big impulse is taken on the server set on fire
  (Utilities|FlowControl|Gate false
    (:Exit
      (if (Networking|IsServer)
        (Utilities|FlowControl|DoOnce false
          (:Completed
            (CallFunction|SetonFire)))))))

(event Custom|Handle%20Received%20Large%20Impulse
  (params (HitComponent :type Primitive%20Component%20Object%20Reference) (OtherActor :type Actor%20Object%20Reference) (OtherComp :type Primitive%20Component%20Object%20Reference) (NormalImpulse :type Vector) (Hit :type Hit%20Result%20Structure%20%28by%20ref%29))
  ;[1 1 1] When a big impulse is taken on the server set on fire
  (Utilities|FlowControl|Gate false
    (:Exit
      (if (Networking|IsServer)
        (Utilities|FlowControl|DoOnce false
          (:Completed
            (CallFunction|SetonFire)))))))

(event EventOnFireHitInfoSet
  (params (Hit :type Hit%20Result%20Structure))
  (bind (_bblockinghit _binitialoverlap _time _distance _location _impactpoint _normal) (Collision|BreakHitResult Hit))
  ;[1 1 1] Thrust while on fire
  (|CallParentFunction Hit)
  (if (Class|AthenaFunctionLibrary|RowtoBool (Variables|Default|GetRow_ThrustEnabled))
    (PhysicsObjects|SetPhysicsObjectPhysicalDataTag self (Variables|Default|GetNoFrictionPresetTag))
    ;;[1 1 1] Add a scene component to use as the location of the thrust then set a timer for when the thrust should begin
    (bind _returnvalue (AddComponent|NODE_AddSceneComponent-2 :RelativeTransform (Math|Transform|MakeTransform (Math|Transform|InverseTransformLocation (Transformation|GetActorTransform self) (CallFunction|PushTowardsCenterOfMass _location)) (Math|Conversions|RotationFromXVector (Math|Transform|InverseTransformDirection (Transformation|GetActorTransform self) _normal)))))
    (Variables|Default|SetThrusterLocation _returnvalue)
    (Utilities|Time|SetTimerbyEvent (delegate Custom|StartThrust) (- (Variables|Default|GetFuseTime) (* (Variables|Default|GetFuseTime) (Variables|Default|GetThrustEndFuseDurationPercent))))))

(event Custom|StartThrust
  ;[1 1 1] Thrust while on fire
  ;;[1 1 1] Start the thrust timer to apply an impulse every frame
  (if (not (HeldObjects|IsHeld (Variables|Default|GetFortHeldObject)))
    (bind _returnvalue (Utilities|Time|SetTimerbyEvent (delegate Custom|Impulse) (Variables|Default|GetImpulseThrustTime) true))
    (Variables|Default|SetTimer_ImpulseThrust _returnvalue)))

(event EventOnShouldDie
  ;[1 1 1] Thrust while on fire
  ;;[1 1 1] Ensure the thrust timer ends
  (Utilities|Time|ClearandInvalidateTimerbyHandle (Variables|Default|GetTimer_ImpulseThrust))
  (|CallParentFunction)
  ;[1 1 1] Check to see if in water and then Get water info
  (Components|Activation|Deactivate (Variables|Default|GetNS_Object_FloatingInWater))
  (bind (_outhit _returnvalue) (Collision|LineTraceForObjects (+ (Transformation|GetActorLocation) "0, 0,200.000000") (- (Transformation|GetActorLocation) "0, 0,200.000000") (Utilities|Array|MakeArray "ObjectTypeQuery12")))
  (if _returnvalue
    (bind _asfort_water_body_actor (Utilities|Casting|CastToFortWaterBodyActor (.HitActor (Collision|BreakHitResult _outhit))))
    (bind _outwatersurfaceinfo (WaterBody|GetWaterSurfaceInfoStruct _asfort_water_body_actor (.Location (Collision|BreakHitResult _outhit)) true false false))
    (bind _returnvalue_2 (Effects|Components|ParticleSystem|SpawnFXAtLocation "/Game/Effects/Niagara/Systems/Enviroment/Water/NS_WaterSplash_Medium.NS_WaterSplash_Medium" (.WaterSurfaceLocation (Utilities|Struct|BreakWaterSurfaceInfo _outwatersurfaceinfo))))
    (Effects|Components|ParticleSystem|SetFloatParameter _returnvalue_2 "User.WaterDepth" (.WaterDepth (Utilities|Struct|BreakWaterSurfaceInfo _outwatersurfaceinfo)))
    (Effects|Components|ParticleSystem|SetVectorParameter _returnvalue_2 "User.WaterPlaneLocation" (Utilities|Struct|BreakWaterSurfaceInfo _outwatersurfaceinfo))
    (Effects|Components|ParticleSystem|SetVectorParameter _returnvalue_2 "User.WaterPlaneNormal" (.WaterPlaneNormal (Utilities|Struct|BreakWaterSurfaceInfo _outwatersurfaceinfo)))
    (Effects|Components|ParticleSystem|SetFloatParameter _returnvalue_2 "User.EffectScale" 2.0)))

(event Custom|Impulse
  ;[1 1 1] Thrust while on fire
  ;;[1 1 1] Impulse the actor at the thruster location with with an inverse forward vector.
  (if (Collision|IsPhysicsCollisionEnabled (Variables|Default|GetParentMesh))
    _)
  (FortPhysicsSimulationLibrary|AddPhysicsImpulseatLocation self (Math|Vector|vector+vector (* (Transformation|GetForwardVector (Variables|Default|GetThrusterLocation)) (* (ScalableFloat|GetValueatLevel (Variables|Default|GetRow_ThrustMag)) -1.0)) (Math|Vector|MakeVector 0.0 0.0 (Variables|Default|GetAddedUp))) (Transformation|GetWorldLocation (Variables|Default|GetThrusterLocation)) false "None" false 0 "(MaxSimSizeAllowed=Large,bRejectEmbeddedInLandscape=False)"))

(event EventOnEnterWater
  (params (Pontoon :type Spherical%20Pontoon%20Structure%20%28by%20ref%29))
  (bind (_centersocket _relativelocation _radius _bfxenabled _localforce _centerlocation _socketrotation _offset _waterheight _waterdepth _immersiondepth _waterplanelocation _waterplanenormal) (Utilities|Struct|BreakSphericalPontoon Pontoon))
  (bind _ns_object_floatinginwater (Variables|Default|GetNS_Object_FloatingInWater))
  (|CallParentFunction Pontoon)
  ;[1 1 1] Get water info if thrown in water
  (Components|Activation|Activate _ns_object_floatinginwater false)
  (Effects|Components|ParticleSystem|SetFloatParameter _ns_object_floatinginwater "User.WaterDepth" _waterdepth)
  (Effects|Components|ParticleSystem|SetVectorParameter _ns_object_floatinginwater "User.WaterPlaneLocation" _waterplanelocation)
  (Effects|Components|ParticleSystem|SetVectorParameter _ns_object_floatinginwater "User.WaterPlaneNormal" _waterplanenormal)
  (Effects|Components|ParticleSystem|SetFloatParameter _ns_object_floatinginwater "User.EffectScale" 2.0)
  (Effects|Components|ParticleSystem|SetFloatParameter _ns_object_floatinginwater "User.InheritVelocityScale" 0.1)
  (Effects|Components|ParticleSystem|SetFloatParameter _ns_object_floatinginwater "User.SpawnCountScale" 3.0))

(fn PushTowardsCenterOfMass
  (params (InputPin :type Vector))
  (return (+ (* (Math|Vector|Normalize (- (Physics|GetCenterOfMass (Variables|Default|GetParentMesh) "None") InputPin)) (* (Math|Vector|VectorLength (- (Physics|GetCenterOfMass (Variables|Default|GetParentMesh) "None") InputPin)) (Variables|Default|GetPushTowardsCenterPercent))) InputPin)))

(fn ConstructionScript
  (|CallParentFunction))
